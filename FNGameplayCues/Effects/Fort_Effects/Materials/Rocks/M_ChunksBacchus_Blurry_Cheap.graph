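# M_ChunksBacchus_Blurry_Cheap is a translucent, unlit particle sprite material, named as a cheap, blurry variant for rock chunk effects.
# Particle color RGB passes through the shared MF_FX_EmissiveAdjustments function to adjust emissive.
# Particle alpha is multiplied by the green channel of the T_Fragment_Bits texture, forming a mask of fragment shapes that follows particle fade.
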Material /FNGameplayCues/Effects/Fort_Effects/Materials/Rocks/M_ChunksBacchus_Blurry_Cheap
graph {
 properties {
  bUsedWithParticleSprites=true
  blendMode=BLEND_Translucent
  pixelDepthOffsetMode=PDOM_Legacy
  shadingModel=MSM_Unlit
  usedShadingModels=Unlit
 }
 nodes [
  n1 = ParticleColor()
  MaterialFunctionCall(Emissive=@n1.RGB) {materialFunction=&MF_FX_EmissiveAdjustments}
  Multiply(A=@n1.A, B=TextureSample() {constMipValue=4, texture=/Game/Effects/Fort_Effects/Textures/Pickup/DimensionalFrag/T_Fragment_Bits}.G)
 ]
}
